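# Physics asset for the TreasureHunter Brigade backpack cosmetic mesh, providing physics bodies for its dangling money pieces.
# A small rounded base body on dyn_money_basa anchors three thin capsule bodies on the dyn_money bones.
# Each money capsule is constrained directly to the base rather than chained, with narrow swing and twist limits that keep its sway tight and vary slightly per piece.
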PhysicsAsset /BRCosmetics/Accessories/FORT_Backpacks/Backpack_F_MED_TreasureHunter_Brigade/Meshes/F_MED_TreasureHunter_Brigade_Pack_FE_Physics
bodies {
 dyn_money_01 [{shapeType=Capsule, length=4, radius=0.5, shapeName=dyn_money_01_capsule, center=<1.6266 0 0>, rotation=<rot 90 0 0>}]
 dyn_money_02 [{shapeType=Capsule, length=4, radius=0.5, shapeName=dyn_money_02_capsule, center=<1.6266 0 0>, rotation=<rot 90 0 0>}]
 dyn_money_03 [{shapeType=Capsule, length=4, radius=0.5, shapeName=dyn_money_03_capsule, center=<1.6266 0 0>, rotation=<rot 90 0 0>}]
 dyn_money_basa [{shapeType=Capsule, length=0, radius=1, shapeName=dyn_money_basa_capsule}]
}
constraints [
 {bone1Name=dyn_money_03, bone2Name=dyn_money_basa, swing1LimitDegrees=20, swing2LimitDegrees=20, twistLimitDegrees=15}
 {bone1Name=dyn_money_01, bone2Name=dyn_money_basa, swing1LimitDegrees=15, swing2LimitDegrees=15, twistLimitDegrees=15}
 {bone1Name=dyn_money_02, bone2Name=dyn_money_basa, swing1LimitDegrees=20, swing2LimitDegrees=30, twistLimitDegrees=15}
]
